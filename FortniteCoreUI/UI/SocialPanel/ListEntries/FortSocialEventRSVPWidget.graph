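# Social panel list-entry widget that displays an event RSVP as skewed, purple-tinted tags for event name, date and additional-event count.
# It starts collapsed on Construct and relies on the FortSocialEventRSVPWidget parent class to supply text through OnRefreshDisplay.
# An empty date collapses the whole widget; otherwise it shows the date and name.
# The additional-events tag collapses separately when its text is empty.
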WidgetBlueprint /FortniteCoreUI/UI/SocialPanel/ListEntries/FortSocialEventRSVPWidget
graphs [EventGraph {nodes [
   n1 = Event Construct()
   SetVisibility(execute=[@n1.then, @n3.then], InVisibility=Collapsed)
   n2 = Event OnRefreshDisplay!() -> (OutputDelegate, then, EventName:"Text (by ref)", Date:"Text (by ref)", AdditionalEvents:"Text (by ref)") {is=title K2Node_Event}
   > n3 = Branch(Condition=TextIsEmpty(InText=@n2.Date))
   SetVisibility(execute=@n3.else, InVisibility=Visible)
   > "SetText (Text)"(self=Get Text_MonthDay() -> (Text_MonthDay:Common Text Object Reference), InText=@n2.Date)
   > "SetText (Text)"(self=Get Text_EventName() -> (Text_EventName:Common Text Object Reference), InText=@n2.EventName)
   > n4 = Branch(Condition=TextIsEmpty(InText=@n2.AdditionalEvents))
   n5 = Get Overlay_AdditionalEvents() -> (Overlay_AdditionalEvents:Overlay Object Reference)
   > SetVisibility(self=@n5, InVisibility=Collapsed)
   SetVisibility(execute=@n4.else, self=@n5, InVisibility=Visible)
   > "SetText (Text)"(self=Get Text_AdditionalEvents() -> (Text_AdditionalEvents:Common Text Object Reference), InText=@n2.AdditionalEvents)
   Event PreConstruct()
   Event Tick()
   Reroute Node(InputPin:Wildcard) -> (OutputPin:Wildcard)
  ]}]
variables [In Visibility {type=ESlateVisibility, default=Visible}]
widgets {
 info {inheritedWidgetCount=0, parentClass=/Script/FortniteUI.FortSocialEventRSVPWidget, rootWidgetClass=&HorizontalBox, widgetCount=14}
 widgets [
  HorizontalBox_0 {widgetClassPath=&HorizontalBox, widget=HorizontalBox_0}
  Overlay_0 {widgetClassPath=&Overlay, parent=HorizontalBox_0, widget=Overlay_0, slot {class=HorizontalBoxSlot}}
  Border_116 {
   widgetClassPath=&Border
   parent=Overlay_0
   widget=Border_116
   properties {
    background {
     bIsDynamicallyLoaded=false
     drawAs=Image
     imageType=NoImage
     mirroring=NoMirror
     resourceName=None
     resourceObject=/Game/UI/SocialPanel/Art/Materials/M_UI_SkewBox_BuffetInfoViolator
     tiling=NoTile
     imageSize=<32 32>
     margin=<ltrb 1 1 1 1>
     outlineSettings {
      bUseBrushTransparency=false
      roundingType=HalfHeightRadius
      width=0
      color {colorUseRule=UseColor_Specified, specifiedColor=<rgba 0 0 0 0>}
      cornerRadii=<0 0 0 0>
     }
     tintColor {colorUseRule=UseColor_Specified, specifiedColor=<rgba 1 1 1 1>}
     uVRegion {bIsValid=false, max=<0 0>, min=<0 0>}
    }
    padding=<ltrb 3 3 3 3>
   }
   slot {class=OverlaySlot, properties {horizontalAlignment=HAlign_Fill, verticalAlignment=VAlign_Fill}}
  }
  Background1 {
   widgetClassPath=&Image
   parent=Border_116
   widget=Background1
   properties {brush {
     bIsDynamicallyLoaded=false
     drawAs=Image
     imageType=NoImage
     mirroring=NoMirror
     resourceName=None
     resourceObject=&M_UI_SkewBox_BuffetInfoViolator
     tiling=NoTile
     imageSize=<32 32>
     margin=<ltrb 0 0 0 0>
     outlineSettings {
      bUseBrushTransparency=false
      roundingType=HalfHeightRadius
      width=0
      color {colorUseRule=UseColor_Specified, specifiedColor=<rgba 0 0 0 0>}
      cornerRadii=<0 0 0 0>
     }
     tintColor {colorUseRule=UseColor_Specified, specifiedColor=<rgba 0.6038 0.0999 0.9911 1>}
     uVRegion {bIsValid=false, max=<0 0>, min=<0 0>}
    }}
   slot {class=BorderSlot, properties {padding=<ltrb 3 3 3 3>}}
  }
  SizeBox_0 {
   widgetClassPath=&SizeBox
   parent=Overlay_0
   widget=SizeBox_0
   properties {bOverride_MaxDesiredWidth=true, maxDesiredWidth=225}
   slot {class=OverlaySlot, properties {horizontalAlignment=HAlign_Center, verticalAlignment=VAlign_Center}}
  }
  ScaleBox_0 {widgetClassPath=&ScaleBox, parent=SizeBox_0, widget=ScaleBox_0, slot {class=SizeBoxSlot, properties {padding=<ltrb 8 1 8 1>}}}
  Text_EventName {
   widgetClassPath=&CommonTextBlock
   bIsVariable=true
   parent=ScaleBox_0
   widget=Text_EventName
   properties {
    text=TBD
    textTransformPolicy=ToUpper
    font {
     bForceMonospaced=false
     bMaterialIsStencil=false
     fontMaterial=null
     fontObject=&BurbankSmall
     letterSpacing=0
     monospacedWidth=1
     size=16
     skewAmount=0
     typefaceFontName=Black
     outlineSettings {
      bApplyOutlineToDropShadows=false
      bMiteredCorners=false
      bSeparateFillAlpha=false
      outlineMaterial=null
      outlineSize=0
      outlineColor=<rgba 0 0 0 1>
     }
    }
    shadowOffset=<0 0>
   }
   slot {class=ScaleBoxSlot}
  }
  Overlay {
   widgetClassPath=&Overlay
   parent=HorizontalBox_0
   widget=Overlay
   slot {class=HorizontalBoxSlot, properties {padding=<ltrb 4 0 4 0>}}
  }
  Background2 {
   widgetClassPath=&Image
   parent=Overlay
   widget=Background2
   properties {brush {
     bIsDynamicallyLoaded=false
     drawAs=Image
     imageType=NoImage
     mirroring=NoMirror
     resourceName=None
     resourceObject=&M_UI_SkewBox_BuffetInfoViolator
     tiling=NoTile
     imageSize=<10 10>
     margin=<ltrb 0 0 0 0>
     outlineSettings {
      bUseBrushTransparency=false
      roundingType=HalfHeightRadius
      width=0
      color {colorUseRule=UseColor_Specified, specifiedColor=<rgba 0 0 0 0>}
      cornerRadii=<0 0 0 0>
     }
     tintColor {colorUseRule=UseColor_Specified, specifiedColor=<rgba 0.6038 0.0999 0.9911 1>}
     uVRegion {bIsValid=false, max=<0 0>, min=<0 0>}
    }}
   slot {class=OverlaySlot, properties {horizontalAlignment=HAlign_Fill, verticalAlignment=VAlign_Fill}}
  }
  HorizontalBox_1 {
   widgetClassPath=&HorizontalBox
   parent=Overlay
   widget=HorizontalBox_1
   slot {class=OverlaySlot, properties {horizontalAlignment=HAlign_Center, verticalAlignment=VAlign_Center}}
  }
  Text_MonthDay {
   widgetClassPath=&CommonTextBlock
   bIsVariable=true
   parent=HorizontalBox_1
   widget=Text_MonthDay
   properties {
    mobileFontSizeMultiplier=1.1
    text=AUG 6
    textTransformPolicy=ToUpper
    colorAndOpacity {colorUseRule=UseColor_Specified, specifiedColor=<rgba 1 1 1 0.75>}
    font {
     bForceMonospaced=false
     bMaterialIsStencil=false
     fontMaterial=null
     fontObject=&BurbankSmall
     letterSpacing=0
     monospacedWidth=1
     size=15
     skewAmount=0
     typefaceFontName=Black
     outlineSettings {
      bApplyOutlineToDropShadows=false
      bMiteredCorners=false
      bSeparateFillAlpha=false
      outlineMaterial=null
      outlineSize=0
      outlineColor=<rgba 0 0 0 1>
     }
    }
    shadowOffset=<0 0>
   }
   slot {class=HorizontalBoxSlot, properties {horizontalAlignment=HAlign_Center, verticalAlignment=VAlign_Center, padding=<ltrb 8 1 8 1>}}
  }
  Overlay_AdditionalEvents {
   widgetClassPath=&Overlay
   bIsVariable=true
   parent=HorizontalBox_0
   widget=Overlay_AdditionalEvents
   slot {class=HorizontalBoxSlot}
  }
  Background3 {
   widgetClassPath=&Image
   parent=Overlay_AdditionalEvents
   widget=Background3
   properties {brush {
     bIsDynamicallyLoaded=false
     drawAs=Image
     imageType=NoImage
     mirroring=NoMirror
     resourceName=None
     resourceObject=/Game/UI/SocialPanel/Art/Materials/M_UI_SkewBox_BuffetInfoViolatorShort
     tiling=NoTile
     imageSize=<10 10>
     margin=<ltrb 0 0 0 0>
     outlineSettings {
      bUseBrushTransparency=false
      roundingType=HalfHeightRadius
      width=0
      color {colorUseRule=UseColor_Specified, specifiedColor=<rgba 0 0 0 0>}
      cornerRadii=<0 0 0 0>
     }
     tintColor {colorUseRule=UseColor_Specified, specifiedColor=<rgba 0.6038 0.0999 0.9911 1>}
     uVRegion {bIsValid=false, max=<0 0>, min=<0 0>}
    }}
   slot {class=OverlaySlot, properties {horizontalAlignment=HAlign_Fill, verticalAlignment=VAlign_Fill}}
  }
  Text_AdditionalEvents {
   widgetClassPath=&CommonTextBlock
   bIsVariable=true
   parent=Overlay_AdditionalEvents
   widget=Text_AdditionalEvents
   properties {
    mobileFontSizeMultiplier=1.1
    text=+1
    textTransformPolicy=ToUpper
    colorAndOpacity {colorUseRule=UseColor_Specified, specifiedColor=<rgba 1 1 1 0.75>}
    font {
     bForceMonospaced=false
     bMaterialIsStencil=false
     fontMaterial=null
     fontObject=&BurbankSmall
     letterSpacing=0
     monospacedWidth=1
     size=15
     skewAmount=0
     typefaceFontName=Black
     outlineSettings {
      bApplyOutlineToDropShadows=false
      bMiteredCorners=false
      bSeparateFillAlpha=false
      outlineMaterial=null
      outlineSize=0
      outlineColor=<rgba 0 0 0 1>
     }
    }
    shadowOffset=<0 0>
   }
   slot {class=OverlaySlot, properties {horizontalAlignment=HAlign_Center, verticalAlignment=VAlign_Center, padding=<ltrb 8 1 8 1>}}
  }
 ]
}
