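# LocalUserSettingsMenu is the social panel's settings menu for the local player, a Widget Blueprint derived from FortLocalUserSettingsMenu.
# A dark blue, fixed-width border holds a vertical stack of four LocalUserSettingSection widgets: online status, privacy, participation and voice channel.
# The sections are inherited variables of the native parent, and participation and voice channel start collapsed.
# The Blueprint only supplies layout, as its event graph contains just unconnected default events.
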WidgetBlueprint /FortniteCoreUI/UI/SocialPanel/LocalUserSettingsMenu
graphs [EventGraph {nodes [
   Event PreConstruct()
   Event Construct()
   Event Tick()
  ]}]
variables []
widgets {
 info {inheritedWidgetCount=4, parentClass=/Script/FortniteUI.FortLocalUserSettingsMenu, rootWidgetClass=&Border, widgetCount=7}
 widgets [
  Border_0 {widgetClassPath=&Border, widget=Border_0, properties {brushColor=<rgba 0 0.0168 0.0887 1>, padding=<ltrb 0 0 0 0>}}
  SizeBox_0 {
   widgetClassPath=&SizeBox
   parent=Border_0
   widget=SizeBox_0
   properties {bOverride_WidthOverride=true, widthOverride=544}
   slot {class=BorderSlot, properties {padding=<ltrb 0 0 0 0>}}
  }
  VerticalBox_0 {
   widgetClassPath=&VerticalBox
   parent=SizeBox_0
   widget=VerticalBox_0
   slot {class=SizeBoxSlot, properties {padding=<ltrb 0 4 0 0>}}
  }
  Section_OnlineStatus {
   widgetClassPath=/Game/UI/SocialPanel/LocalUserSettingSection
   bInherited=true
   bIsVariable=true
   parent=VerticalBox_0
   widget=Section_OnlineStatus
   slot {class=VerticalBoxSlot}
  }
  Section_Privacy {
   widgetClassPath=&LocalUserSettingSection
   bInherited=true
   bIsVariable=true
   parent=VerticalBox_0
   widget=Section_Privacy
   slot {class=VerticalBoxSlot}
  }
  Section_Participation {
   widgetClassPath=&LocalUserSettingSection
   bInherited=true
   bIsVariable=true
   parent=VerticalBox_0
   widget=Section_Participation
   properties {visibility=Collapsed}
   slot {class=VerticalBoxSlot}
  }
  Section_VoiceChannel {
   widgetClassPath=&LocalUserSettingSection
   bInherited=true
   bIsVariable=true
   parent=VerticalBox_0
   widget=Section_VoiceChannel
   properties {visibility=Collapsed}
   slot {class=VerticalBoxSlot}
  }
 ]
}
